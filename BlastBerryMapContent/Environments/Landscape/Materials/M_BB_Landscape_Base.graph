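# Nanite-enabled base landscape material for the BlastBerry map, blending rock, grass, sand, forest, mud, road and gravel using painted weights, slope and height blends.
# The layered result feeds a Runtime Virtual Texture output, with switches selecting sampled RVT or direct layers.
# A mobile shading path replaces full blending with cheaper base-color lerps and derived specular.
# Layer and slope masks also drive procedural foliage, while weightmaps assign physical materials.
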Material /BlastBerryMapContent/Environments/Landscape/Materials/M_BB_Landscape_Base
graph {
 properties {
  bUseMaterialAttributes=true
  bUsedWithNanite=true
  pixelDepthOffsetMode=PDOM_Legacy
  refractionMethod=RM_IndexOfRefraction
  usedShadingModels=Default Lit
 }
 nodes [
  n1 = RuntimeVirtualTextureSample!(UVs, World Position) {
   materialType=BaseColor_Normal_Specular_YCoCg
   virtualTexture=/BlastBerryMapContent/Game/Athena/Environments/Materials/RVT/Athena_RuntimeVirtualTexture
  }
  n2 = SetMaterialAttributes!(
   MaterialAttributes
   Base Color=@n1.BaseColor
   Specular=@n1.BaseColor
   Roughness=@n1.BaseColor
   Normal=Transform(None=@n1.Normal) {transformSourceType=TRANSFORMSOURCE_World, transformType=TRANSFORM_Tangent}
  ) {attributeSetTypes [
    #1
    #2
    #3
    #4
   ]}
  n3 = MaterialFunctionCall!(UseSimple, In=LandscapeLayerCoords() {mappingScale=4}) {materialFunction=/BlastBerryMapContent/Game/Athena/Environments/Landscape/MaterialFunctions/Standard/MF_Athena_Road_01}
  n4 = NamedRerouteDeclaration(None=@n3.Result) {name=Road_Mat, nodeColor=<rgba 0.3359 0 1 1>}
  n5 = MaterialFunctionCall!(In=LandscapeLayerCoords() {mappingScale=4}) {materialFunction=/BlastBerryMapContent/Game/Athena/Environments/Landscape/MaterialFunctions/Standard/MF_Athena_Mud_01}
  n6 = NamedRerouteUsage()
  n7 = MaterialFunctionCall!(UVs=LandscapeLayerCoords() {mappingScale=4}) {materialFunction=/BlastBerryMapContent/Game/Athena/Environments/Landscape/MaterialFunctions/Standard/MF_Athena_Forest_01}
  n8 = NamedRerouteUsage()
  n9 = MaterialFunctionCall!(InputMaterial=@n6, LayerMaterial=@n7.Result, LayerWeight=@n8, UseContrast, Contrast) {materialFunction=/Game/Athena/Helios/Environments/Landscape/Materials/MaterialFunctions_Utilities/MF_HeightBlendLandscapeLayers_OpacityMask}
  n10 = MaterialFunctionCall(
   InputMaterial=@n9.BlendedMaterial
   LayerMaterial=@n5.Result
   LayerWeight=NamedRerouteUsage()
   UseContrast=StaticBool() {value=true}
   Contrast=Constant() {r=1}
  ) {materialFunction=&MF_HeightBlendLandscapeLayers_OpacityMask}
  n11 = MaterialFunctionCall(
   InputMaterial=@n10.BlendedMaterial
   LayerMaterial=@n4
   LayerWeight=NamedRerouteUsage()
   UseContrast=StaticBool() {value=true}
   Contrast=ScalarParameter() {defaultValue=0.5, group=Road, parameterName=RoadBlendContrast}
  ) {materialFunction=&MF_HeightBlendLandscapeLayers_OpacityMask}
  n12 = LandscapeLayerSwitch!(LayerUsed=@n11.BlendedMaterial, LayerNotUsed=@n10.BlendedMaterial) {parameterName=Road}
  n13 = GetMaterialAttributes(None=MaterialFunctionCall!(Base Material=@n12, Top Material=NamedRerouteUsage(), Alpha=NamedRerouteUsage()) {materialFunction=/Engine/Functions/MaterialLayerFunctions/MatLayerBlend_Standard}.Blended Material) {attributeGetTypes [
    #1
    #2
    #3
    #4
   ]}
  n14 = SetMaterialAttributes(Base Color=@n13.Base Color, Specular=@n13.Base Color, Roughness=@n13.Base Color, Normal=@n13.Base Color) {attributeSetTypes [
    #1
    #2
    #3
    #4
   ]}
  n15 = GetMaterialAttributes(None=NamedRerouteUsage()) {attributeGetTypes [#1]}
  n16 = NamedRerouteUsage()
  n17 = NamedRerouteUsage()
  n18 = NamedRerouteUsage()
  n19 = LinearInterpolate(A=@n18, B=NamedRerouteUsage(), Alpha=NamedRerouteUsage())
  n20 = Multiply(
   A=LinearInterpolate(A=NamedRerouteUsage(), B=LandscapeLayerSwitch(LayerUsed=@n19, LayerNotUsed=@n18) {parameterName=Sand}, Alpha=NamedRerouteUsage())
   B=NamedRerouteUsage() {desc=New}
  )
  n21 = NamedRerouteUsage()
  n22 = LinearInterpolate(
   A=LinearInterpolate(A=@n20, B=Multiply(A=@n20, B=@n21), Alpha=NamedRerouteUsage())
   B=NamedRerouteUsage()
   Alpha=MaterialFunctionCall(In=NamedRerouteUsage(), Contrast=Constant() {r=2}) {materialFunction=&CheapContrast}.Result
  )
  n23 = LinearInterpolate(A=@n22, B=@n17, Alpha=NamedRerouteUsage())
  n24 = GetMaterialAttributes(None=NamedRerouteUsage()) {attributeGetTypes [#1]}
  n25 = LinearInterpolate(
   A=LandscapeLayerSwitch(LayerUsed=LinearInterpolate(A=@n23, B=@n15.Base Color, Alpha=@n16), LayerNotUsed=@n23) {parameterName=Road}
   B=@n24.Base Color
   Alpha=NamedRerouteUsage()
  )
  n26 = Multiply(A=ComponentMask(None=@n25) {r=true}, B=ScalarParameter() {defaultValue=2, parameterName=SimpleMobileSpec})
  n27 = GetMaterialAttributes(None=SetMaterialAttributes!(
    MaterialAttributes
    Base Color=@n25
    Specular=Saturate(None=@n26)
    Roughness=ScalarParameter() {defaultValue=0.6, parameterName=SimpleMobileRough}
   ) {attributeSetTypes [#1, #2, #3]}) {attributeGetTypes [
    #1
    #2
    #3
    #4
   ]}
  n28 = SetMaterialAttributes!(MaterialAttributes, Base Color=@n27.Base Color, Specular=@n27.Base Color, Roughness=@n27.Base Color, Normal=@n27.Base Color) {attributeSetTypes [
    #1
    #2
    #3
    #4
   ]}
  n29 = ShadingPathSwitch(Default=@n14, Mobile=@n28)
  n30 = MaterialFunctionCall(NoRVT=@n29, RVT=@n2) {
   desc=used to check for switch as well
   materialFunction=&FN_RVT_Switch_Material
  }
  n31 = StaticSwitchParameter(True=@n29, False=@n30.Result) {group=Rufus, parameterName=RVT_Off}
  n32 = GetMaterialAttributes(None=@n31) {attributeGetTypes [#1]}
  n33 = LinearInterpolate(
   A=MaterialFunctionCall() {materialFunction=/BlastBerryMapContent/Environments/Landscape/Materials/MaterialFunctions/MF_BB_BiomeMask}.Mountain
   B=NamedRerouteUsage()
   Alpha=ScalarParameter() {parameterName=sfdg346terg}
  )
  BlendMaterialAttributes(
   A=@n32.MaterialAttributes
   B=SetMaterialAttributes!(MaterialAttributes=@n32.MaterialAttributes, Base Color=Multiply(A=@n33)) {attributeSetTypes [#1]}
   Alpha=ScalarParameter() {defaultValue=1, parameterName=DEBUG_AsBaseColor}
  )
  Constant() {r=0.8}
  n34 = MaterialFunctionCall!(Material=NamedRerouteUsage()) {materialFunction=/Engine/Functions/MaterialLayerFunctions/MatLayerBlend_BreakNormal}
  n35 = Constant3Vector() {constant=<rgba 0 0 1 0>}
  n36 = LandscapeLayerSample() {parameterName=Road, previewWeight=1}
  n37 = LandscapeLayerSample() {parameterName=Mud, previewWeight=1}
  n38 = Clamp(None=Add(A=Add(A=@n36, B=@n37), B=LandscapeLayerSample() {parameterName=Gravel, previewWeight=1}))
  n39 = Divide(A=WorldPosition().XY, B=ScalarParameter() {defaultValue=25000, parameterName=MacroNormalScale})
  n40 = ScalarParameter() {defaultValue=0.625, parameterName=MacroNormalIntensity}
  n41 = Constant() {r=0.25}
  n42 = Constant() {r=0.9}
  n43 = FeatureLevelSwitch(Default=@n41, ES2=@n42, ES3_1=@n42, SM4_REMOVED=@n41)
  n44 = Multiply(
   A=LinearInterpolate(
    A=Constant3Vector() {constant=<rgba 0 0 1 0>}
    B=TextureSample(UVs=@n39) {samplerType=SAMPLERTYPE_Normal, texture=/Game/ContentCreationTools/TextureCreation/Textures/T_BPCreated_MacroNormal_01}.RGB
    Alpha=MaterialFunctionCall!(Distance) {materialFunction=/Game/Athena/Environments/Landscape/MaterialFunctions/MF_TerrainDistanceFade}.Result
   )
   B=MaterialFunctionCall(X=@n40, Y=@n40, Z=QualitySwitch(Default=@n43, Low=@n42, Medium=@n42)) {materialFunction=&MakeFloat3}.Result
  )
  n45 = Add(A=Multiply(A=LinearInterpolate(A=@n34.Normal, B=@n35, Alpha=@n38), B=Constant3Vector() {constant=<rgba 1 1 0.5 0>}), B=@n44)
  n46 = QualitySwitch(Default=@n45, Low=@n44, High=@n45, Medium=@n44)
  FeatureLevelSwitch(Default=@n46, ES2=@n44, ES3_1=@n44, SM4_REMOVED=@n46, SM5=@n46, SM6=@n46)
  n47 = OneMinus(None=ComponentMask(None=VertexNormalWS()) {b=true})
  n48 = Constant() {r=5}
  n49 = Saturate(None=Add(A=Multiply(A=@n47, B=@n48), B=Constant() {r=-0.1}))
  n50 = NamedRerouteUsage()
  n51 = NamedRerouteUsage()
  n52 = NamedRerouteUsage()
  n53 = Add(A=Add(A=Add(A=@n50, B=@n49), B=@n51), B=Ceil(None=@n52))
  n54 = NamedRerouteUsage()
  n55 = Add(A=Add(A=Add(A=@n53, B=@n54), B=NamedRerouteUsage()), B=Round(None=NamedRerouteUsage()))
  n56 = Subtract(A=Subtract(A=Constant() {r=1}, B=Round(None=@n49)), B=@n55)
  MaterialExpressionReroute_32 = Reroute(None=Saturate(None=@n56))
  n57 = Subtract(A=NamedRerouteUsage(), B=@n55)
  MaterialExpressionReroute_35 = Reroute(None=Saturate(None=@n57))
  n58 = NamedRerouteUsage()
  n59 = Add(A=Add(A=@n51, B=@n52), B=@n54)
  MaterialExpressionSaturate_17 = Saturate(None=Subtract(A=Subtract(A=@n58, B=@n59), B=Round(None=@n49)))
  n60 = NamedRerouteUsage()
  MaterialExpressionSaturate_15 = Saturate(None=Subtract(A=Subtract(A=@n60, B=@n49), B=NamedRerouteUsage()))
  MaterialExpressionSaturate_23 = Saturate(None=Subtract(A=NamedRerouteUsage(), B=NamedRerouteUsage()))
  MaterialExpressionSaturate_13 = Saturate(None=Add(A=@n56, B=@n57))
  MaterialExpressionMax_0 = Max(A=@MaterialExpressionReroute_32, B=@MaterialExpressionReroute_35)
  LandscapeGrassOutput!(
   Grass=@MaterialExpressionReroute_32
   Mountain=@MaterialExpressionReroute_35
   Forest=@MaterialExpressionSaturate_17
   Gravel=@MaterialExpressionSaturate_15
   Flowers=@MaterialExpressionSaturate_23
   Bushes=@MaterialExpressionSaturate_13
   TallPlant=@MaterialExpressionMax_0
  ) {grassTypes [
    Grass {
     grassType=/BlastBerryMapContent/Game/Athena/Environments/Landscape/Foliage/Athena_Foliage_Grass
     input {
      expression=&self:MaterialExpressionReroute_32
      inputName=None
      mask=0
      maskA=0
      maskB=0
      maskG=0
      maskR=0
      outputIndex=0
     }
    }
    Mountain {
     grassType=/BlastBerryMapContent/Game/Athena/Environments/Landscape/Foliage/Athena_Foliage_Mountain_Grass1
     input {
      expression=&self:MaterialExpressionReroute_35
      inputName=None
      mask=0
      maskA=0
      maskB=0
      maskG=0
      maskR=0
      outputIndex=0
     }
    }
    Forest {
     grassType=/BlastBerryMapContent/Game/Athena/Environments/Landscape/Foliage/Athena_Foliage_Forest
     input {
      expression=&self:MaterialExpressionSaturate_17
      inputName=None
      mask=0
      maskA=0
      maskB=0
      maskG=0
      maskR=0
      outputIndex=0
     }
    }
    Gravel {
     grassType=/Game/Athena/Environments/Landscape/Foliage/Athena_Foliage_Gravel
     input {
      expression=&self:MaterialExpressionSaturate_15
      inputName=None
      mask=0
      maskA=0
      maskB=0
      maskG=0
      maskR=0
      outputIndex=0
     }
    }
    Flowers {
     grassType=/BlastBerryMapContent/Game/Athena/Environments/Landscape/Foliage/Rufus_Foliage_Flowers
     input {
      expression=&self:MaterialExpressionSaturate_23
      inputName=None
      mask=0
      maskA=0
      maskB=0
      maskG=0
      maskR=0
      outputIndex=0
     }
    }
    Bushes {
     grassType=/BlastBerryMapContent/Game/Athena/Environments/Landscape/Foliage/Athena_Foliage_Bushes
     input {
      expression=&self:MaterialExpressionSaturate_13
      inputName=None
      mask=0
      maskA=0
      maskB=0
      maskG=0
      maskR=0
      outputIndex=0
     }
    }
    TallPlant {
     grassType=/BlastBerryMapContent/Game/Athena/Environments/Landscape/Foliage/Rufus_TallPlant
     input {
      expression=&self:MaterialExpressionMax_0
      inputName=None
      mask=0
      maskA=0
      maskB=0
      maskG=0
      maskR=0
      outputIndex=0
     }
    }
   ]}
  MaterialExpressionNamedRerouteUsage_82 = NamedRerouteUsage()
  MaterialExpressionNamedRerouteUsage_77 = NamedRerouteUsage()
  MaterialExpressionNamedRerouteUsage_169 = NamedRerouteUsage()
  MaterialExpressionNamedRerouteUsage_87 = NamedRerouteUsage()
  MaterialExpressionNamedRerouteUsage_88 = NamedRerouteUsage()
  LandscapePhysicalMaterialOutput!(
   Stone=@MaterialExpressionNamedRerouteUsage_82
   Dirt=@MaterialExpressionNamedRerouteUsage_77
   Asphalt=@MaterialExpressionNamedRerouteUsage_169
   Sand=@MaterialExpressionNamedRerouteUsage_87
   Grass=@MaterialExpressionNamedRerouteUsage_88
  ) {
   desc="These are layered in order of importance.\r\nIf Dirt=1 and Stone=1, then Stone will \"win\".\r\nTheoretically, these values aren't clamped.\r\nSo a value of 2 should trump a value of 1."
   inputs [
    {
     physicalMaterial=/Game/FortressPhysicalMaterials/Stone
     input {
      expression=&self:MaterialExpressionNamedRerouteUsage_82
      inputName=None
      mask=0
      maskA=0
      maskB=0
      maskG=0
      maskR=0
      outputIndex=0
     }
    }
    {
     physicalMaterial=/Game/FortressPhysicalMaterials/Dirt
     input {
      expression=&self:MaterialExpressionNamedRerouteUsage_77
      inputName=None
      mask=0
      maskA=0
      maskB=0
      maskG=0
      maskR=0
      outputIndex=0
     }
    }
    {
     physicalMaterial=/Game/FortressPhysicalMaterials/Asphalt
     input {
      expression=&self:MaterialExpressionNamedRerouteUsage_169
      inputName=None
      mask=0
      maskA=0
      maskB=0
      maskG=0
      maskR=0
      outputIndex=0
     }
    }
    {
     physicalMaterial=/Game/FortressPhysicalMaterials/Sand
     input {
      expression=&self:MaterialExpressionNamedRerouteUsage_87
      inputName=None
      mask=0
      maskA=0
      maskB=0
      maskG=0
      maskR=0
      outputIndex=0
     }
    }
    {
     physicalMaterial=/Game/FortressPhysicalMaterials/Grass
     input {
      expression=&self:MaterialExpressionNamedRerouteUsage_88
      inputName=None
      mask=0
      maskA=0
      maskB=0
      maskG=0
      maskR=0
      outputIndex=0
     }
    }
   ]
  }
  LinearInterpolate(
   A=@n32.Base Color
   B=Multiply(A=@n32.Base Color, B=VectorParameter() {parameterName=DEBUG_AsBaseColorColor, defaultValue=<rgba 0 0 0.75 1>}.RGB)
  )
  n61 = NamedRerouteUsage()
  n62 = Saturate(None=NamedRerouteUsage())
  n63 = LinearInterpolate(
   A=Saturate(None=NamedRerouteUsage())
   B=Saturate(None=NamedRerouteUsage())
   Alpha=ScalarParameter() {group=Cliff, parameterName=SnowCliff_DEBUGMe5, sortPriority=98}
  )
  n64 = ScalarParameter() {defaultValue=1, group=Cliff, parameterName=SnowCliff_DEBUGMe4, sortPriority=98}
  LinearInterpolate(
   A=Saturate(None=NamedRerouteUsage())
   B=LinearInterpolate(
    A=Saturate(None=@n61)
    B=LinearInterpolate(A=@n62, B=@n63, Alpha=@n64)
    Alpha=ScalarParameter() {defaultValue=1, group=Cliff, parameterName=SnowCliff_DEBUGMe3, sortPriority=98}
   )
   Alpha=ScalarParameter() {group=Cliff, parameterName=SnowCliff_DEBUGMe2, sortPriority=98}
  )
  n65 = BreakMaterialAttributes!(Attr)
  MakeMaterialAttributes!(
   BaseColor=@n65.BaseColor
   Metallic
   Specular=Multiply(A=@n65.Specular)
   Roughness=Multiply(A=@n65.Roughness)
   Anisotropy
   EmissiveColor
   Opacity
   OpacityMask
   Normal=Multiply(A=@n65.Normal)
   Tangent
   WorldPositionOffset
   SubsurfaceColor=Constant()
   ClearCoat
   ClearCoatRoughness
   AmbientOcclusion
   Refraction
   CustomizedUVs_0
   CustomizedUVs_1
   CustomizedUVs_2
   CustomizedUVs_3
   CustomizedUVs_4
   CustomizedUVs_5
   CustomizedUVs_6
   CustomizedUVs_7
   PixelDepthOffset
   ShadingModel
   Displacement
  )
  n66 = MaterialFunctionCall!(In=LandscapeLayerCoords() {mappingScale=4}) {materialFunction=/BlastBerryMapContent/Game/Athena/Environments/Landscape/MaterialFunctions/Standard/MF_Athena_Gravel_01}
  MaterialFunctionCall(
   InputMaterial=@n12
   LayerMaterial=NamedRerouteDeclaration(None=@n66.Result) {name=Gravel_Mat, nodeColor=<rgba 1 0.3281 0 1>}
   LayerWeight=NamedRerouteUsage()
   UseContrast=StaticBool() {value=true}
   Contrast=ScalarParameter() {defaultValue=0.5, group=Gravel, parameterName=GravelLayer_BlendContrast}
  ) {materialFunction=&MF_HeightBlendLandscapeLayers_OpacityMask}
  MaterialFunctionCall!(In=SetMaterialAttributes!(MaterialAttributes=@n31, Opacity Mask=LandscapeVisibilityMask()) {attributeSetTypes [#5]}) {materialFunction=/Game/Athena/Asteria/Environments/Landscape/Materials/MaterialFunctions/MF_Terrain_LumenModification}
  n67 = WorldPosition()
  n68 = Constant() {r=-32.1449}
  n69 = Saturate(None=Divide(A=Subtract(A=@n67.Z, B=@n68) {constB=-50}, B=Constant() {r=20}) {constB=50})
  MaterialFunctionCall!(
   Material
   NewBaseColor=Constant() {r=0.7}
   Mask=MaterialFunctionCall(Transition Phase=OneMinus(None=@n69), Height Texture=NamedRerouteUsage()) {materialFunction=&HeightLerp}.Alpha
  ) {materialFunction=/Engine/Functions/MaterialLayerFunctions/MatLayerBlend_MultiplyBaseColor}
  MaterialFunctionCall(
   Base Material=@n6
   Top Material=@n7.Result
   Alpha=NamedRerouteDeclaration(None=@n9.LayerMask) {name=Forest_BlendMask, nodeColor=<rgba 0.1953 0 1 1>}
  ) {materialFunction=&MatLayerBlend_Standard}
  MaterialFunctionCall(Transition Phase=@n8, Height Texture=@n7.Height, Contrast=Constant() {r=0.25}) {materialFunction=&HeightLerp}
  MaterialFunctionCall(Base Material=@n9.BlendedMaterial, Top Material=NamedRerouteUsage(), Alpha=NamedRerouteUsage()) {materialFunction=&MatLayerBlend_Standard}
  n70 = MaterialFunctionCall() {materialFunction=/BlastBerryMapContent/Game/Athena/Environments/Landscape/MaterialFunctions/Standard/MF_Athena_Rock_01}
  NamedRerouteDeclaration(None=@n70.BaseColor_Plain) {name=Rock_BaseColorPlain, nodeColor=<rgba 1 0 0.2812 1>}
  n71 = MaterialFunctionCall() {materialFunction=/BlastBerryMapContent/Game/Athena/Environments/Landscape/MaterialFunctions/Standard/MF_Athena_Grass_01}
  NamedRerouteDeclaration(None=@n71.BaseColor_Plain) {name=Grass_BaseColorPlain, nodeColor=<rgba 1 0.1875 0 1>}
  NamedRerouteDeclaration(None=LandscapeLayerSample() {parameterName=Gravel, previewWeight=1}) {name=Gravel_Weight, nodeColor=<rgba 0 0 0 1>}
  n72 = GetMaterialAttributes(None=MaterialFunctionCall() {materialFunction=/BlastBerryMapContent/Environments/Landscape/Materials/MaterialFunctions/MF_BB_Sand_01}.Result) {attributeGetTypes [#1, #4, #5]}
  NamedRerouteDeclaration(None=@n72.MaterialAttributes) {name=Sand_Mat, nodeColor=<rgba 0.4531 1 0 1>}
  n73 = CollectionParameter() {collection=/BlastBerryMapContent/Game/Athena/Environments/Landscape/MPC/MPC_Landscape, parameterName=Rock_Bottom_Brightness}
  n74 = CollectionParameter() {collection=&MPC_Landscape, parameterName=Rock_Top_Brightness}
  n75 = Constant()
  n76 = MaterialFunctionCall(
   Transition Phase=MaterialFunctionCall!(height_Offset, height_GradScale, Coords=WorldPosition().XYZ) {materialFunction=/Game/Athena/Environments/Landscape/MaterialFunctions/MF_Athena_WorldHeightGrad}.Result
   Height Texture=@n75
   Contrast=@n75
  ) {materialFunction=&HeightLerp}
  NamedRerouteDeclaration(None=MaterialFunctionCall!(In=@n70.Result, Color=LinearInterpolate(A=@n73, B=@n74, Alpha=@n76.Alpha)) {materialFunction=/Game/Athena/Environments/Landscape/MaterialFunctions/MF_Athena_Material_BaseMultiply}.Result) {name=Rock_Mat, nodeColor=<rgba 0.6641 0 1 1>}
  NamedRerouteDeclaration(None=@n71.Result) {name=Grass_Mat, nodeColor=<rgba 0 1 0.3438 1>}
  NamedRerouteDeclaration(None=LandscapeLayerSample() {parameterName=Sand}) {name=Sand_Weight, nodeColor=<rgba 0 0 0 1>}
  NamedRerouteDeclaration(None=LandscapeLayerSample() {parameterName=Mud}) {name=Mud_Weight, nodeColor=<rgba 0 0 0 1>}
  NamedRerouteDeclaration(None=GetMaterialAttributes(None=@n70.Result) {attributeGetTypes [#6]}.Opacity) {name=Rock_Alpha, nodeColor=<rgba 0 0.2266 1 1>}
  n77 = CollectionParameter() {collection=&MPC_Landscape, parameterName=RockTint}
  NamedRerouteDeclaration(None=LinearInterpolate(
    A=ComponentMask(None=@n77) {b=true, g=true, r=true}
    B=ScalarParameter() {defaultValue=0.8, group=Grass, parameterName=Grass_Tint}
    Alpha=NamedRerouteUsage()
   )) {name=GrassRock_ColorAdjustment, nodeColor=<rgba 0 0.3906 1 1>}
  NamedRerouteDeclaration(None=@n72.Opacity Mask) {name=Sand_Height, nodeColor=<rgba 0.0078 1 0 1>}
  n78 = WorldPosition()
  n79 = Constant() {r=-50}
  n80 = Saturate(None=Divide(A=Subtract(A=@n78.Z, B=@n79) {constB=-50}, B=Constant() {r=20}) {constB=50})
  NamedRerouteDeclaration(None=MaterialFunctionCall(Transition Phase=OneMinus(None=@n80), Height Texture=NamedRerouteUsage()) {materialFunction=&HeightLerp}.Alpha) {name=Sand_WetHeight, nodeColor=<rgba 1 0.9609 0 1>}
  n81 = Multiply(A=ComponentMask(None=VertexNormalWS()) {b=true}, B=Constant() {r=0.859})
  n82 = Add(A=@n81, B=Constant() {r=-0.01})
  n83 = NamedRerouteUsage()
  n84 = NamedRerouteUsage()
  n85 = LandscapeLayerSwitch(
   LayerUsed=MaterialFunctionCall(Transition Phase=NamedRerouteUsage(), Height Texture=OneMinus(None=@n84), Contrast=Constant() {r=1}) {materialFunction=&HeightLerp}.Alpha
   LayerNotUsed=Constant()
  ) {parameterName=Sand}
  n86 = LandscapeLayerSwitch(
   LayerUsed=BlendMaterialAttributes(A=@n83, B=NamedRerouteUsage(), Alpha=NamedRerouteDeclaration(None=@n85) {name=Sand_BlendMask, nodeColor=<rgba 0.8516 1 0 1>})
   LayerNotUsed=@n83
  ) {parameterName=Sand}
  n87 = Multiply(
   A=Multiply(A=GetMaterialAttributes(None=@n86) {attributeGetTypes [#5]}.Opacity Mask, B=NamedRerouteUsage())
   B=Constant() {r=0.08}
  )
  n88 = Constant() {r=5}
  NamedRerouteDeclaration(None=MaterialFunctionCall(Transition Phase=Saturate(None=@n82), Height Texture=@n87, Contrast=@n88) {materialFunction=&HeightLerp}.Alpha) {name=GrassRock_BlendAlpha_Bias, nodeColor=<rgba 0 1 0.1562 1>}
  NamedRerouteDeclaration(None=GetMaterialAttributes(None=@n7.Result) {attributeGetTypes [#1]}.Base Color) {name=Forest_BaseColor, nodeColor=<rgba 0 1 0.8359 1>}
  NamedRerouteDeclaration(None=GetMaterialAttributes(None=@n5.Result) {attributeGetTypes [#1]}.Base Color) {name=Mud_BaseColor, nodeColor=<rgba 1 0.5156 0 1>}
  NamedRerouteDeclaration(None=@n10.LayerMask) {name=Mud_BlendAlpha, nodeColor=<rgba 0.1953 0 1 1>}
  NamedRerouteDeclaration(None=@n5.Result) {name=Mud_Mat, nodeColor=<rgba 0.125 1 0 1>}
  NamedRerouteDeclaration(None=LandscapeLayerSample() {parameterName=Forest}) {name=Forest_Weight, nodeColor=<rgba 0 0 0 1>}
  NamedRerouteDeclaration(None=LandscapeLayerSwitch(LayerUsed=@n11.LayerMask, LayerNotUsed=Constant()) {parameterName=Road}) {name=Road_BlendMask, nodeColor=<rgba 1 0 0.3047 1>}
  n89 = MaterialFunctionCall(Transition Phase=Saturate(None=@n81), Height Texture=@n87, Contrast=@n88) {materialFunction=&HeightLerp}
  NamedRerouteDeclaration(None=@n89.Lerp Alpha No Contrast) {name=RockGrassBlend_NoContrast, nodeColor=<rgba 1 0.1172 0 1>}
  n90 = MaterialFunctionCall(Transition Phase=NamedRerouteUsage(), Height Texture=@n66.Height, Contrast=Constant() {r=2}) {materialFunction=&HeightLerp}
  NamedRerouteDeclaration(None=@n90.Lerp Alpha No Contrast) {name=Gravel_BlendMask, nodeColor=<rgba 0.0078 0 1 1>}
  NamedRerouteDeclaration(None=LandscapeLayerSample() {parameterName=Road, previewWeight=1}) {name=Road_Weight, nodeColor=<rgba 0 0 0 1>}
  n91 = NamedRerouteUsage()
  n92 = Power(Base=@n91)
  n93 = MaterialFunctionCall!(Mat attributes=@n3.Result) {materialFunction=/Engine/Functions/Engine_MaterialFunctions02/MatLayerBlend_BreakOpacity}
  n94 = Constant() {r=5}
  NamedRerouteDeclaration(None=MaterialFunctionCall(
    Base Material=NamedRerouteUsage()
    Top Material=@n3.Result
    Alpha=MaterialFunctionCall(Transition Phase=@n92, Height Texture=@n93.Result, Contrast=@n94) {materialFunction=&HeightLerp}.Alpha
   ) {materialFunction=&MatLayerBlend_Standard}.Blended Material) {name=RoadMud_Mat, nodeColor=<rgba 1 0 0.3984 1>}
  NamedRerouteDeclaration(None=LandscapeLayerSample() {parameterName=Mountain_Grass}) {name=MountainGrass_Weight, nodeColor=<rgba 0 0 0 1>}
  NamedRerouteDeclaration(None=LandscapeLayerSample() {parameterName=NoFoliage}) {name=NoFoliage_Weight, nodeColor=<rgba 0 0 0 1>}
  NamedRerouteDeclaration(None=LandscapeLayerSample() {parameterName=Flowers, previewWeight=1}) {name=Flowers_Weight, nodeColor=<rgba 0 0 0 1>}
  NamedRerouteDeclaration(None=Multiply(A=NamedRerouteUsage(), B=Constant() {r=3})) {name=Asphalt_physmat_Weightmap, nodeColor=<rgba 0.5 0 1 1>}
  NamedRerouteDeclaration(None=@n71.Height) {name=GrassHeight, nodeColor=<rgba 1 0 0.8437 1>}
  n95 = Multiply(A=NamedRerouteUsage(), B=OneMinus(None=NamedRerouteUsage()))
  n96 = NamedRerouteUsage()
  NamedRerouteDeclaration(None=Multiply(A=Max(A=@n95, B=@n96), B=Constant() {r=4})) {name=Dirt_Physmat_WeightMap, nodeColor=<rgba 0.5 0 1 1>}
  n97 = NamedRerouteUsage()
  n98 = Constant() {r=5}
  NamedRerouteDeclaration(None=OneMinus(None=Multiply(A=@n97, B=@n98))) {name=Stone_Physmat_Weightmap, nodeColor=<rgba 0.5 0 1 1>}
  NamedRerouteDeclaration(None=@n72.Normal) {name=Sand_Normal, nodeColor=<rgba 1 0 0.3281 1>}
  NamedRerouteDeclaration(None=Constant() {r=1}) {name=Grass_Physmat_Weightmap, nodeColor=<rgba 0.5 0 1 1>}
  n99 = OneMinus(None=NamedRerouteUsage())
  n100 = NamedRerouteUsage()
  NamedRerouteDeclaration(None=Multiply(A=Multiply(A=@n99, B=@n100), B=Constant() {r=2})) {name=Sand_Physmat_Weightmap, nodeColor=<rgba 0.5 0 1 1>}
  n101 = NamedRerouteUsage()
  n102 = MaterialFunctionCall(
   Base Material=NamedRerouteUsage()
   Top Material=@n86
   Alpha=NamedRerouteDeclaration(None=@n89.Alpha) {name=GrassRock_BlendAlpha, nodeColor=<rgba 1 0 0.0703 1>}
  ) {materialFunction=&MatLayerBlend_Standard}
  n103 = NamedRerouteUsage()
  n104 = NamedRerouteDeclaration(None=Saturate(None=MaterialFunctionCall!(Rock_Diffuse, ColorBlend_Alpha, ColorBlend_Alpha_Bias=@n103, Rock_Diffuse1, Grass_Diffuse) {materialFunction=/BlastBerryMapContent/Environments/Landscape/Materials/MaterialFunctions/MF_Mountain_Colors}.Blended_Colors_Bias)) {name=Mountain_Colors, nodeColor=<rgba 0.9688 1 0 1>}
  n105 = Subtract(A=NamedRerouteUsage(), B=Multiply(A=NamedRerouteUsage(), B=NamedRerouteUsage()))
  NamedRerouteDeclaration(None=BlendMaterialAttributes(
    A=MaterialFunctionCall(In=@n102.Blended Material, Color=@n101) {materialFunction=&MF_Athena_Material_BaseMultiply}.Result
    B=MaterialFunctionCall(In=@n102.Blended Material, Color=@n104) {materialFunction=&MF_Athena_Material_BaseMultiply}.Result
    Alpha=Saturate(None=@n105)
   )) {name=BaseBlend, nodeColor=<rgba 1 0 0.4453 1>}
  NamedRerouteDeclaration(None=@n72.Base Color) {name=Sand_BaseColorPlain, nodeColor=<rgba 0.8281 0 1 1>}
  NamedRerouteDeclaration(None=@n90.Alpha) {name=Gravel_BlendMaskContrast, nodeColor=<rgba 0.0078 0 1 1>}
  NamedRerouteUsage()
  NamedRerouteUsage()
  NamedRerouteUsage()
  NamedRerouteUsage()
  NamedRerouteUsage()
  n106 = Constant() {r=-60}
  n107 = Constant() {r=140}
  n108 = Add(A=NamedRerouteUsage(), B=Constant() {r=-0.24})
  OneMinus(None=Saturate(None=LinearInterpolate(A=@n106, B=@n107, Alpha=@n108)))
  Reroute(None=MaterialFunctionCall(Base Material=@n10.BlendedMaterial, Top Material=@n4, Alpha=NamedRerouteUsage()) {materialFunction=&MatLayerBlend_Standard}.Blended Material)
  n109 = Constant() {r=1.2}
  n110 = OneMinus(None=GetMaterialAttributes(None=@n10.BlendedMaterial) {attributeGetTypes [#5]}.Opacity Mask)
  n111 = Constant() {r=0.25}
  Reroute(None=MaterialFunctionCall(Transition Phase=Multiply(A=@n91, B=@n109), Height Texture=Multiply(A=@n110, B=@n111), Contrast=Constant() {r=5}) {materialFunction=&HeightLerp}.Alpha)
  RuntimeVirtualTextureOutput!(
   BaseColor=@n13.Base Color
   Specular=@n13.Base Color
   Roughness=@n13.Base Color
   Normal=Transform(None=@n13.Normal)
   WorldHeight=WorldPosition().Z
   Opacity
   Mask
   Displacement
   Mask4
  )
  n112 = NamedRerouteUsage()
  n113 = OneMinus(None=GetMaterialAttributes(None=@n9.BlendedMaterial) {attributeGetTypes [#5]}.Opacity Mask)
  n114 = Constant() {r=1}
  n115 = NamedRerouteUsage()
  n116 = ScalarParameter() {defaultValue=0.4, parameterName=Remove_DirtFromSand_OceanFix}
  Saturate(None=Subtract(
    A=MaterialFunctionCall(Transition Phase=@n112, Height Texture=@n113, Contrast=@n114) {materialFunction=&HeightLerp}.Alpha
    B=Multiply(A=@n115, B=@n116) {constB=0.5}
   ))
  n117 = LinearInterpolate(A=@n22, B=@n17, Alpha=NamedRerouteUsage())
  SetMaterialAttributes!(
   MaterialAttributes
   Base Color=LinearInterpolate(A=LinearInterpolate(A=@n117, B=@n15.Base Color, Alpha=@n16), B=@n24.Base Color, Alpha=NamedRerouteUsage())
  ) {attributeSetTypes [#1]}
  n118 = GetMaterialAttributes() {attributeGetTypes [#1]}
  n119 = VectorParameter() {group=Debug, parameterName=DEBUG-TraversalMaskColor, defaultValue=<rgba 1 0 0 1>}
  n120 = ScalarParameter() {desc=Set this to 1 to turn it on, group=Debug, parameterName=DEBUG-TraversalMask}
  n121 = Constant() {r=43}
  n122 = Constant() {r=-16.5}
  n123 = WorldPosition()
  n124 = CrossProduct(A=DDX(Value=@n123.XYZ), B=DDY(Value=@n123.XYZ))
  n125 = Abs(None=ComponentMask(None=Normalize(VectorInput=@n124)) {b=true})
  n126 = Constant() {r=1}
  n127 = If(A=Saturate(None=LinearInterpolate(A=@n121, B=@n122, Alpha=@n125)), B=Constant() {r=0.75}, "A > B"=@n126, "A == B"=@n126, "A < B"=Constant())
  SetMaterialAttributes(
   MaterialAttributes=@n118.MaterialAttributes
   Base Color=LinearInterpolate(A=@n118.Base Color, B=Multiply(A=@n118.Base Color, B=@n119.RGB), Alpha=Multiply(A=@n120, B=@n127))
  ) {attributeSetTypes [#1]}
  StaticSwitchParameter(True=BlendMaterialAttributes(A=@n29, B=@n2, Alpha=LandscapeLayerSample() {parameterName=Grass_Only}), False=@n2) {group=Rufus, parameterName=RVT_LayerSetup}
  StaticSwitchParameter(False=@n65.Specular) {defaultValue=true, parameterName=UseCheapSpec}
 ]
}
